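# One-shot mono MetaSound source for the BondShock "cook out" sound, apparently tied to a head throw, played as player foley.
# On play, a random-get node picks one of two wave variations and feeds it to a wave player, whose finish ends the source.
# It uses a generic player-equip attenuation, a short max distance, and the BondShock default concurrency set.
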MetaSoundSource /WildEstate/Gameplay/BondShock/Assets/Audio/SFX/MSS_BondShock_Cook_Out
inputs [
 Sounds {type=WaveAsset:Array}
 UE.Source.OnPlay {type=Trigger}
]
interfaces [UE.OutputFormat.Mono:1.0, UE.Source.OneShot:1.0, UE.Source:1.1]
outputs [
 "UE.OutputFormat.Mono.Audio:0" {type=Audio}
 UE.Source.OneShot.OnFinished {type=Trigger}
]
pages [Default {
  input_defaults {
   UE.Source.OnPlay=null
   Sounds [
    /WildEstate/Gameplay/BondShock/Assets/Audio/SFX/wav/BondShock_Head_Throw_Cook_Out_01
    /WildEstate/Gameplay/BondShock/Assets/Audio/SFX/wav/BondShock_Head_Throw_Cook_Out_02
   ]
  }
  nodes [
   ArrayRandomGet_0 = Array.Random Get:WaveAsset:Array(In Array=^Sounds, Next=^UE.Source.OnPlay)
   WavePlayer_0 = UE.Wave Player:Mono(Play=@ArrayRandomGet_0.On Next, Wave Asset=@ArrayRandomGet_0.Value)
  ]
  links [{dst="@UE.Source.OneShot.OnFinished", src=WavePlayer_0.On Finished}, {dst="@UE.OutputFormat.Mono.Audio:0", src=WavePlayer_0.Out Mono}]
 }]
properties {
 assetImportData=&self:AssetImportData
 attenuationSettings=/Game/Sounds/Attenuation/Fort_Player_Equip_Generic_Att
 maxDistance=1800
 sampleRate=48000
 soundClassObject=&Fort_Player_Foley
 concurrencySet [/WildEstate/Gameplay/BondShock/Assets/Audio/CONC_BondShock_Default]
}
