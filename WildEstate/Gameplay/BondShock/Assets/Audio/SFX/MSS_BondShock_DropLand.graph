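# One-shot mono MetaSound for the BondShock head drop-landing sound, routed as player foley with attenuation and concurrency settings.
# On play, two random-get nodes each pick a landing impact and a roll variation, played simultaneously by separate wave players.
# A gain-boosted two-input mixer combines them.
# The finish trigger waits on the impact's start and the roll's end, so the roll alone sets the sound's length.
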MetaSoundSource /WildEstate/Gameplay/BondShock/Assets/Audio/SFX/MSS_BondShock_DropLand
inputs [
 Roll {type=WaveAsset:Array}
 Sounds {type=WaveAsset:Array}
 UE.Source.OnPlay {type=Trigger}
]
interfaces [UE.OutputFormat.Mono:1.0, UE.Source.OneShot:1.0, UE.Source:1.1]
outputs [
 "UE.OutputFormat.Mono.Audio:0" {type=Audio}
 UE.Source.OneShot.OnFinished {type=Trigger}
]
pages [Default {
  input_defaults {
   UE.Source.OnPlay=null
   Roll [
    /WildEstate/Gameplay/BondShock/Assets/Audio/SFX/wav/BondShock_Head_Roll_01
    /WildEstate/Gameplay/BondShock/Assets/Audio/SFX/wav/BondShock_Head_Roll_02
    /WildEstate/Gameplay/BondShock/Assets/Audio/SFX/wav/BondShock_Head_Roll_03
    /WildEstate/Gameplay/BondShock/Assets/Audio/SFX/wav/BondShock_Head_Roll_04
   ]
   Sounds [
    /WildEstate/Gameplay/BondShock/Assets/Audio/SFX/wav/BondShock_Head_DropLand_01
    /WildEstate/Gameplay/BondShock/Assets/Audio/SFX/wav/BondShock_Head_DropLand_02
    /WildEstate/Gameplay/BondShock/Assets/Audio/SFX/wav/BondShock_Head_DropLand_03
   ]
  }
  nodes [
   ArrayRandomGet_0 = Array.Random Get:WaveAsset:Array(In Array=^Sounds, Next=^UE.Source.OnPlay)
   WavePlayer_0 = UE.Wave Player:Mono(Play=@ArrayRandomGet_0.On Next, Wave Asset=@ArrayRandomGet_0.Value) {also {version [1,1]}}
   ArrayRandomGet_1 = Array.Random Get:WaveAsset:Array(In Array=^Roll, Next=^UE.Source.OnPlay) {also {defaults {Weights []}}}
   WavePlayer_1 = UE.Wave Player:Mono(Play=@ArrayRandomGet_1.On Next, Wave Asset=@ArrayRandomGet_1.Value) {also {version [1,1]}}
   AudioMixerAudioMixerMono2_0 = "AudioMixer.Audio Mixer (Mono, 2)"(In 0=@WavePlayer_0.Out Mono, In 1=@WavePlayer_1.Out Mono) {also {defaults {Gain 0=3, Gain 1=2}}}
   TriggerAccumulatorTriggerAccumulate2_0 = "TriggerAccumulator.Trigger Accumulate (2)"(In 0=@WavePlayer_0.On Play, In 1=@WavePlayer_1.On Finished)
  ]
  links [
   {dst="@UE.OutputFormat.Mono.Audio:0", src=AudioMixerAudioMixerMono2_0.Out}
   {dst="@UE.Source.OneShot.OnFinished", src=TriggerAccumulatorTriggerAccumulate2_0.Out}
  ]
 }]
properties {
 assetImportData=&self:AssetImportData
 attenuationSettings=/Game/Sounds/Attenuation/Fort_Player_Equip_Generic_Att
 maxDistance=1800
 sampleRate=48000
 soundClassObject=&Fort_Player_Foley
 concurrencySet [/WildEstate/Gameplay/BondShock/Assets/Audio/CONC_BondShock_Default]
}
