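# One-shot quad MetaSound for the first-person BondShock helmet hit, with attenuation, concurrency and a send to a weapon sidechain bus.
# A tonal impact plays immediately in the front pair, followed by briefly delayed, slightly staggered random left and right stereo layers split front-to-surround on their sides.
# All three players share one random pitch shift per play.
# VolumeMultiplier scales every output, and OnFinished waits for all three layers.
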MetaSoundSource /WildEstate/Gameplay/BondShock/Assets/Audio/SFX/MSS_BondShock_HelmetHit_1P
inputs [
 Left {type=WaveAsset:Array}
 Main {type=WaveAsset:Array}
 Right {type=WaveAsset:Array}
 UE.Source.OnPlay {type=Trigger}
 VolumeMultiplier {type=Float}
]
interfaces [UE.OutputFormat.Quad:1.0, UE.Source.OneShot:1.0, UE.Source:1.1]
outputs [
 "UE.OutputFormat.Quad.Audio:0" {type=Audio}
 "UE.OutputFormat.Quad.Audio:1" {type=Audio}
 "UE.OutputFormat.Quad.Audio:4" {type=Audio}
 "UE.OutputFormat.Quad.Audio:5" {type=Audio}
 UE.Source.OneShot.OnFinished {type=Trigger}
]
pages [Default {
  input_defaults {
   UE.Source.OnPlay=null
   VolumeMultiplier=1
   Left [
    /WildEstate/Gameplay/BondShock/Assets/Audio/SFX/wav/BondShock_Quad_L_OS_01
    /WildEstate/Gameplay/BondShock/Assets/Audio/SFX/wav/BondShock_Quad_L_OS_02
    /WildEstate/Gameplay/BondShock/Assets/Audio/SFX/wav/BondShock_Quad_L_OS_03
   ]
   Main [/WildEstate/Gameplay/BondShock/Assets/Audio/SFX/wav/BondShock_Statue_Impt_Tonal_Close]
   Right [
    /WildEstate/Gameplay/BondShock/Assets/Audio/SFX/wav/BondShock_Quad_R_OS_01
    /WildEstate/Gameplay/BondShock/Assets/Audio/SFX/wav/BondShock_Quad_R_OS_02
    /WildEstate/Gameplay/BondShock/Assets/Audio/SFX/wav/BondShock_Quad_R_OS_03
   ]
  }
  nodes [
   TriggerDelay_0 = UE.Trigger Delay(In=^UE.Source.OnPlay) {also {defaults {Delay Time=0.2}}}
   ArrayRandomGet_2 = Array.Random Get:WaveAsset:Array(In Array=^Right, Next=@TriggerDelay_0.Out) {also {defaults {Weights []}}}
   n1 = UE.RandomFloat(Next=^UE.Source.OnPlay)
   n2 = VariableAccessor.Float(Variable=VariableMutator.Float(Variable=InitVariable.Float().Variable, Value=@n1.Value).Variable)
   WavePlayer_1 = UE.Wave Player:Stereo(Play=@ArrayRandomGet_2.On Next, Wave Asset=@ArrayRandomGet_2.Value, Pitch Shift=@n2.Value) {also {version [1,1]}}
   ArrayRandomGet_0 = Array.Random Get:WaveAsset:Array(In Array=^Main, Next=@n1.On Next)
   n3 = VariableAccessor.Float(Variable=@n2.Variable)
   WavePlayer_0 = UE.Wave Player:Stereo(Play=@ArrayRandomGet_0.On Next, Wave Asset=@ArrayRandomGet_0.Value, Pitch Shift=VariableAccessor.Float(Variable=@n3.Variable).Value) {also {version [1,1]}}
   AudioMixerAudioMixerMono2_0 = "AudioMixer.Audio Mixer (Mono, 2)"(Gain 0=^VolumeMultiplier, Gain 1=^VolumeMultiplier, In 1=UE.Reroute(Value=@WavePlayer_1.Out Left).Value, In 0=@WavePlayer_0.Out Right)
   TriggerDelay_1 = UE.Trigger Delay(In=^UE.Source.OnPlay) {also {defaults {Delay Time=0.22}}}
   ArrayRandomGet_1 = Array.Random Get:WaveAsset:Array(In Array=^Left, Next=@TriggerDelay_1.Out) {also {defaults {Weights []}}}
   WavePlayer_2 = UE.Wave Player:Stereo(Play=@ArrayRandomGet_1.On Next, Wave Asset=@ArrayRandomGet_1.Value, Pitch Shift=@n3.Value) {also {version [1,1]}}
   AudioMixerAudioMixerMono2_1 = "AudioMixer.Audio Mixer (Mono, 2)"(Gain 0=^VolumeMultiplier, Gain 1=^VolumeMultiplier, In 0=@WavePlayer_0.Out Left, In 1=@WavePlayer_2.Out Left)
   Multiply_0 = UE.Multiply:Audio by Float(AdditionalOperands=^VolumeMultiplier, PrimaryOperand=UE.Reroute(Value=@WavePlayer_2.Out Right).Value)
   Multiply_1 = UE.Multiply:Audio by Float(AdditionalOperands=^VolumeMultiplier, PrimaryOperand=UE.Reroute(Value=@WavePlayer_1.Out Right).Value)
   TriggerAccumulatorTriggerAccumulate3_0 = "TriggerAccumulator.Trigger Accumulate (3)"(In 0=@WavePlayer_0.On Finished, In 2=@WavePlayer_1.On Finished, In 1=@WavePlayer_2.On Finished)
  ]
  links [
   {dst="@UE.OutputFormat.Quad.Audio:1", src=AudioMixerAudioMixerMono2_0.Out}
   {dst="@UE.OutputFormat.Quad.Audio:0", src=AudioMixerAudioMixerMono2_1.Out}
   {dst="@UE.OutputFormat.Quad.Audio:4", src=Multiply_0.Out}
   {dst="@UE.OutputFormat.Quad.Audio:5", src=Multiply_1.Out}
   {dst="@UE.Source.OneShot.OnFinished", src=TriggerAccumulatorTriggerAccumulate3_0.Out}
  ]
 }]
properties {
 assetImportData=&self:AssetImportData
 attenuationSettings=/WildEstate/Gameplay/BondShock/Assets/Audio/ATT_BondShock_Helmet_Impact_1P
 maxDistance=2500
 numChannels=4
 outputFormat=Quad
 sampleRate=48000
 soundClassObject=&Fort_Player_Foley
 busSends [{audioBus=&AB_Sidechain_FortWeapon_Vol, sendLevel=0.5}]
 concurrencySet [/WildEstate/Gameplay/BondShock/Assets/Audio/CONC_BondShock_Default]
}
