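# One-shot stereo MetaSound for BondShock's third-person helmet hit, playing thrown-head impact sounds that vary with distance.
# On play, it picks random close (stereo) and distant (mono) waves, each with a random pitch shift.
# Attenuation distance is range-mapped to crossfade both channels from close to distant.
# A trigger accumulator signals finish once both players end, while attenuation and concurrency settings govern falloff and voice count.
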MetaSoundSource /WildEstate/Gameplay/BondShock/Assets/Audio/SFX/MSS_BondShock_HelmetHit_3P
inputs [
 Close Impacts {type=WaveAsset:Array}
 Distant Impacts {type=WaveAsset:Array}
 UE.Attenuation.Distance {type=Float}
 UE.Source.OnPlay {type=Trigger}
]
interfaces [UE.Attenuation:1.0, UE.OutputFormat.Stereo:1.0, UE.Source.OneShot:1.0, UE.Source:1.1]
outputs [
 "UE.OutputFormat.Stereo.Audio:0" {type=Audio}
 "UE.OutputFormat.Stereo.Audio:1" {type=Audio}
 UE.Source.OneShot.OnFinished {type=Trigger}
]
pages [Default {
  input_defaults {
   UE.Attenuation.Distance=28440.0117
   UE.Source.OnPlay=null
   Close Impacts [
    /WildEstate/Gameplay/BondShock/Assets/Audio/SFX/wav/BondShock_Thrown_Head_Impt_Close_01
    /WildEstate/Gameplay/BondShock/Assets/Audio/SFX/wav/BondShock_Thrown_Head_Impt_Close_02
    /WildEstate/Gameplay/BondShock/Assets/Audio/SFX/wav/BondShock_Thrown_Head_Impt_Close_03
   ]
   Distant Impacts [
    /WildEstate/Gameplay/BondShock/Assets/Audio/SFX/wav/BondShock_Thrown_Head_Impt_Dist_01
    /WildEstate/Gameplay/BondShock/Assets/Audio/SFX/wav/BondShock_Thrown_Head_Impt_Dist_02
    /WildEstate/Gameplay/BondShock/Assets/Audio/SFX/wav/BondShock_Thrown_Head_Impt_Dist_03
   ]
  }
  nodes [
   n1 = MapRange.MapRange:Float(In=^UE.Attenuation.Distance) {also {defaults {In Range B=30000}}}
   ArrayRandomGet_0 = Array.Random Get:WaveAsset:Array(In Array=^Distant Impacts, Next=^UE.Source.OnPlay) {also {defaults {Weights []}}}
   n2 = UE.RandomFloat(Next=@ArrayRandomGet_0.On Next) {also {defaults {Max=-3, Min=1.1}}}
   WavePlayer_0 = UE.Wave Player:Mono(Wave Asset=@ArrayRandomGet_0.Value, Play=@n2.On Next, Pitch Shift=@n2.Value) {also {version [1,1]}}
   ArrayRandomGet_1 = Array.Random Get:WaveAsset:Array(In Array=^Close Impacts, Next=^UE.Source.OnPlay)
   n3 = UE.RandomFloat(Next=@ArrayRandomGet_1.On Next) {also {defaults {Max=-3, Min=1.1}}}
   WavePlayer_1 = UE.Wave Player:Stereo(Wave Asset=@ArrayRandomGet_1.Value, Play=@n3.On Next, Pitch Shift=@n3.Value) {also {version [1,1]}}
   CrossfadeTriggerRouteAudio2_0 = "Crossfade.Trigger Route (Audio, 2):Audio"(Crossfade Value=@n1.Out Value, In 1=@WavePlayer_0.Out Mono, In 0=@WavePlayer_1.Out Right)
   CrossfadeTriggerRouteAudio2_1 = "Crossfade.Trigger Route (Audio, 2):Audio"(Crossfade Value=@n1.Out Value, In 1=@WavePlayer_0.Out Mono, In 0=@WavePlayer_1.Out Left)
   TriggerAccumulatorTriggerAccumulate2_0 = "TriggerAccumulator.Trigger Accumulate (2)"(In 1=@WavePlayer_0.On Finished, In 0=@WavePlayer_1.On Finished)
  ]
  links [
   {dst="@UE.OutputFormat.Stereo.Audio:1", src=CrossfadeTriggerRouteAudio2_0.Out}
   {dst="@UE.OutputFormat.Stereo.Audio:0", src=CrossfadeTriggerRouteAudio2_1.Out}
   {dst="@UE.Source.OneShot.OnFinished", src=TriggerAccumulatorTriggerAccumulate2_0.Out}
  ]
 }]
properties {
 assetImportData=&self:AssetImportData
 attenuationSettings=/WildEstate/Gameplay/BondShock/Assets/Audio/ATT_BondShock_Helmet_Impact_3P
 bOverrideConcurrency=true
 maxDistance=31000
 numChannels=2
 outputFormat=Stereo
 sampleRate=48000
 soundClassObject=&Fort_Player_Foley
 concurrencyOverrides {maxCount=4, retriggerTime=0.5}
 concurrencySet [/WildEstate/Gameplay/BondShock/Assets/Audio/CONC_BondShock_Default]
}
