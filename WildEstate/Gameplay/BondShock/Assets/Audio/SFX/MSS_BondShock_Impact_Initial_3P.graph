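# One-shot stereo MetaSound for the third-person initial impact of BondShock, spatialized with a helmet-impact attenuation and routed to the Fort impacts sound class.
# On play, it picks a random wave from the Sounds array, which by default holds a single head-impact clip, and plays it through a stereo Wave Player.
# Both channels are scaled by VolumeMultiplier, and the source finishes when playback ends.
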MetaSoundSource /WildEstate/Gameplay/BondShock/Assets/Audio/SFX/MSS_BondShock_Impact_Initial_3P
inputs [
 Sounds {type=WaveAsset:Array}
 UE.Source.OnPlay {type=Trigger}
 VolumeMultiplier {type=Float}
]
interfaces [UE.OutputFormat.Stereo:1.0, UE.Source.OneShot:1.0, UE.Source:1.1]
outputs [
 "UE.OutputFormat.Stereo.Audio:0" {type=Audio}
 "UE.OutputFormat.Stereo.Audio:1" {type=Audio}
 UE.Source.OneShot.OnFinished {type=Trigger}
]
pages [Default {
  input_defaults {UE.Source.OnPlay=null, VolumeMultiplier=1.5, Sounds [/WildEstate/Gameplay/BondShock/Assets/Audio/SFX/wav/BondShock_Head_ImpactFromInteract]}
  nodes [
   ArrayRandomGet_0 = Array.Random Get:WaveAsset:Array(In Array=^Sounds, Next=^UE.Source.OnPlay)
   WavePlayer_0 = UE.Wave Player:Stereo(Play=@ArrayRandomGet_0.On Next, Wave Asset=@ArrayRandomGet_0.Value) {also {version [1,1]}}
   Multiply_0 = UE.Multiply:Audio by Float(AdditionalOperands=^VolumeMultiplier, PrimaryOperand=@WavePlayer_0.Out Right)
   Multiply_1 = UE.Multiply:Audio by Float(AdditionalOperands=^VolumeMultiplier, PrimaryOperand=@WavePlayer_0.Out Left)
  ]
  links [
   {dst="@UE.OutputFormat.Stereo.Audio:1", src=Multiply_0.Out}
   {dst="@UE.OutputFormat.Stereo.Audio:0", src=Multiply_1.Out}
   {dst="@UE.Source.OneShot.OnFinished", src=WavePlayer_0.On Finished}
  ]
 }]
properties {
 assetImportData=&self:AssetImportData
 attenuationSettings=/WildEstate/Gameplay/BondShock/Assets/Audio/ATT_BondShock_Helmet_Impact_3P
 maxDistance=31000
 numChannels=2
 outputFormat=Stereo
 sampleRate=48000
 soundClassObject=&Fort_Impacts
 volume=0.8
 concurrencySet [/WildEstate/Gameplay/BondShock/Assets/Audio/CONC_BondShock_Default]
}
