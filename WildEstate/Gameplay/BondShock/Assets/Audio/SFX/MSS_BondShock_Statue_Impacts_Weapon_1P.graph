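# One-shot stereo MetaSound for first-person weapon impacts on BondShock statues.
# On play, it fires a random weapon-impact wave, then a random statue ring-out wave after a short trigger delay.
# A mixer sums both layers, with the Multiplier input setting the impact gain, and OnFinished fires once both players finish.
# It uses the impacts sound class, a 1P no-volume-attenuation asset and BondShock impact concurrency.
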MetaSoundSource /WildEstate/Gameplay/BondShock/Assets/Audio/SFX/MSS_BondShock_Statue_Impacts_Weapon_1P
inputs [
 Multiplier {type=Float}
 RingOut {type=WaveAsset:Array}
 Sounds {type=WaveAsset:Array}
 UE.Source.OnPlay {type=Trigger}
]
interfaces [UE.OutputFormat.Stereo:1.0, UE.Source.OneShot:1.0, UE.Source:1.1]
outputs [
 "UE.OutputFormat.Stereo.Audio:0" {type=Audio}
 "UE.OutputFormat.Stereo.Audio:1" {type=Audio}
 UE.Source.OneShot.OnFinished {type=Trigger}
]
pages [Default {
  input_defaults {
   Multiplier=2.5
   UE.Source.OnPlay=null
   RingOut [
    /WildEstate/Gameplay/BondShock/Assets/Audio/SFX/wav/BondShock_Impact_Statue_01
    /WildEstate/Gameplay/BondShock/Assets/Audio/SFX/wav/BondShock_Impact_Statue_02
    /WildEstate/Gameplay/BondShock/Assets/Audio/SFX/wav/BondShock_Impact_Statue_03
    /WildEstate/Gameplay/BondShock/Assets/Audio/SFX/wav/BondShock_Impact_Statue_04
   ]
   Sounds [
    /WildEstate/Gameplay/BondShock/Assets/Audio/SFX/wav/BondShock_Impact_Weapon_Statue_01
    /WildEstate/Gameplay/BondShock/Assets/Audio/SFX/wav/BondShock_Impact_Weapon_Statue_02
    /WildEstate/Gameplay/BondShock/Assets/Audio/SFX/wav/BondShock_Impact_Weapon_Statue_03
    /WildEstate/Gameplay/BondShock/Assets/Audio/SFX/wav/BondShock_Impact_Weapon_Statue_04
   ]
  }
  nodes [
   TriggerDelay_0 = UE.Trigger Delay(In=^UE.Source.OnPlay) {also {defaults {Delay Time=0.05}}}
   ArrayRandomGet_1 = Array.Random Get:WaveAsset:Array(In Array=^RingOut, Next=@TriggerDelay_0.Out) {also {defaults {Weights []}}}
   WavePlayer_0 = UE.Wave Player:Stereo(Play=@ArrayRandomGet_1.On Next, Wave Asset=@ArrayRandomGet_1.Value) {also {version [1,1]}}
   ArrayRandomGet_0 = Array.Random Get:WaveAsset:Array(In Array=^Sounds, Next=^UE.Source.OnPlay)
   WavePlayer_1 = UE.Wave Player:Stereo(Play=@ArrayRandomGet_0.On Next, Wave Asset=@ArrayRandomGet_0.Value) {also {version [1,1]}}
   AudioMixerAudioMixerStereo2_0 = "AudioMixer.Audio Mixer (Stereo, 2)"(
    Gain 0=^Multiplier
    In 1 L=@WavePlayer_0.Out Left
    In 1 R=@WavePlayer_0.Out Right
    In 0 L=@WavePlayer_1.Out Left
    In 0 R=@WavePlayer_1.Out Right
   ) {also {defaults {Gain 1=3.5}}}
   TriggerAccumulatorTriggerAccumulate2_0 = "TriggerAccumulator.Trigger Accumulate (2)"(In 1=@WavePlayer_0.On Finished, In 0=@WavePlayer_1.On Finished)
  ]
  links [
   {dst="@UE.OutputFormat.Stereo.Audio:0", src=AudioMixerAudioMixerStereo2_0.Out L}
   {dst="@UE.OutputFormat.Stereo.Audio:1", src=AudioMixerAudioMixerStereo2_0.Out R}
   {dst="@UE.Source.OneShot.OnFinished", src=TriggerAccumulatorTriggerAccumulate2_0.Out}
  ]
 }]
properties {
 assetImportData=&self:AssetImportData
 attenuationSettings=/WildEstate/Gameplay/BondShock/Assets/Audio/ATT_BondShock_Helmet_Impact_1P_NoVolAtt
 numChannels=2
 outputFormat=Stereo
 sampleRate=48000
 soundClassObject=&Fort_Impacts
 concurrencyOverrides {maxCount=5}
 concurrencySet [/WildEstate/Gameplay/BondShock/Assets/Audio/CONC_BondShock_Impact]
}
