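# One-shot stereo MetaSound for third-person weapon impacts on the BondShock statue.
# On play, a random getter picks one of four impact waves and feeds a wave player that drives both output channels and signals finished when playback ends.
# It plays in the impacts sound class with statue attenuation and a BondShock impact concurrency set.
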MetaSoundSource /WildEstate/Gameplay/BondShock/Assets/Audio/SFX/MSS_BondShock_Statue_Impacts_Weapon_3P
inputs [
 Sounds {type=WaveAsset:Array}
 UE.Source.OnPlay {type=Trigger}
]
interfaces [UE.OutputFormat.Stereo:1.0, UE.Source.OneShot:1.0, UE.Source:1.1]
outputs [
 "UE.OutputFormat.Stereo.Audio:0" {type=Audio}
 "UE.OutputFormat.Stereo.Audio:1" {type=Audio}
 UE.Source.OneShot.OnFinished {type=Trigger}
]
pages [Default {
  input_defaults {
   UE.Source.OnPlay=null
   Sounds [
    /WildEstate/Gameplay/BondShock/Assets/Audio/SFX/wav/BondShock_Impact_Weapon_Statue_01
    /WildEstate/Gameplay/BondShock/Assets/Audio/SFX/wav/BondShock_Impact_Weapon_Statue_02
    /WildEstate/Gameplay/BondShock/Assets/Audio/SFX/wav/BondShock_Impact_Weapon_Statue_03
    /WildEstate/Gameplay/BondShock/Assets/Audio/SFX/wav/BondShock_Impact_Weapon_Statue_04
   ]
  }
  nodes [
   ArrayRandomGet_0 = Array.Random Get:WaveAsset:Array(In Array=^Sounds, Next=^UE.Source.OnPlay)
   WavePlayer_0 = UE.Wave Player:Stereo(Play=@ArrayRandomGet_0.On Next, Wave Asset=@ArrayRandomGet_0.Value) {also {version [1,1]}}
  ]
  links [
   {dst="@UE.Source.OneShot.OnFinished", src=WavePlayer_0.On Finished}
   {dst="@UE.OutputFormat.Stereo.Audio:0", src=WavePlayer_0.Out Left}
   {dst="@UE.OutputFormat.Stereo.Audio:1", src=WavePlayer_0.Out Right}
  ]
 }]
properties {
 assetImportData=&self:AssetImportData
 attenuationSettings=/WildEstate/Gameplay/BondShock/Assets/Audio/ATT_BondShock_Statue
 maxDistance=3300
 numChannels=2
 outputFormat=Stereo
 sampleRate=48000
 soundClassObject=&Fort_Impacts
 concurrencyOverrides {maxCount=5, retriggerTime=0.1}
 concurrencySet [/WildEstate/Gameplay/BondShock/Assets/Audio/CONC_BondShock_Impact]
}
