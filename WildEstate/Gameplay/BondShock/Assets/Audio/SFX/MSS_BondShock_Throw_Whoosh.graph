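# One-shot stereo MetaSound Source for the BondShock throw whoosh, classed as player foley.
# On play it picks a random head-throw whoosh wave, then rolls a random pitch shift before triggering the wave player, whose finish ends the sound.
# Both channels are scaled by VolumeMultiplier, with the player's right output feeding output 0 and left feeding output 1, swapping the stereo image.
# It borrows the first-person helmet impact attenuation and uses default BondShock concurrency.
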MetaSoundSource /WildEstate/Gameplay/BondShock/Assets/Audio/SFX/MSS_BondShock_Throw_Whoosh
inputs [
 Sounds {type=WaveAsset:Array}
 UE.Source.OnPlay {type=Trigger}
 VolumeMultiplier {type=Float}
]
interfaces [UE.OutputFormat.Stereo:1.0, UE.Source.OneShot:1.0, UE.Source:1.1]
outputs [
 "UE.OutputFormat.Stereo.Audio:0" {type=Audio}
 "UE.OutputFormat.Stereo.Audio:1" {type=Audio}
 UE.Source.OneShot.OnFinished {type=Trigger}
]
pages [Default {
  input_defaults {
   UE.Source.OnPlay=null
   VolumeMultiplier=2
   Sounds [
    /WildEstate/Gameplay/BondShock/Assets/Audio/SFX/wav/BondShock_Head_Throw_Whoosh_01
    /WildEstate/Gameplay/BondShock/Assets/Audio/SFX/wav/BondShock_Head_Throw_Whoosh_02
   ]
  }
  nodes [
   ArrayRandomGet_0 = Array.Random Get:WaveAsset:Array(In Array=^Sounds, Next=^UE.Source.OnPlay)
   n1 = UE.RandomFloat(Next=@ArrayRandomGet_0.On Next) {also {defaults {Min=-3}}}
   WavePlayer_0 = UE.Wave Player:Stereo(Wave Asset=@ArrayRandomGet_0.Value, Play=@n1.On Next, Pitch Shift=@n1.Value)
   Multiply_0 = UE.Multiply:Audio by Float(AdditionalOperands=^VolumeMultiplier, PrimaryOperand=@WavePlayer_0.Out Right)
   Multiply_1 = UE.Multiply:Audio by Float(AdditionalOperands=^VolumeMultiplier, PrimaryOperand=@WavePlayer_0.Out Left)
  ]
  links [
   {dst="@UE.OutputFormat.Stereo.Audio:0", src=Multiply_0.Out}
   {dst="@UE.OutputFormat.Stereo.Audio:1", src=Multiply_1.Out}
   {dst="@UE.Source.OneShot.OnFinished", src=WavePlayer_0.On Finished}
  ]
 }]
properties {
 assetImportData=&self:AssetImportData
 attenuationSettings=/WildEstate/Gameplay/BondShock/Assets/Audio/ATT_BondShock_Helmet_Impact_1P
 maxDistance=2500
 numChannels=2
 outputFormat=Stereo
 sampleRate=48000
 soundClassObject=&Fort_Player_Foley
 concurrencySet [/WildEstate/Gameplay/BondShock/Assets/Audio/CONC_BondShock_Default]
}
